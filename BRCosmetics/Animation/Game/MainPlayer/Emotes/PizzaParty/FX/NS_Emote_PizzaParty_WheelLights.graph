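# Niagara system for the PizzaParty emote's vehicle lights: red back light, bluish headlights, red ribbon trail and ground smoke.
# Lights and trail follow skeletal mesh sockets and use a system fade-in, with camera-angle-driven glows faking directionality.
# In gameplay, lights show only at night, while smoke varies by time of day and, like the trail, spawns by distance moved.
# Front-end modifiers instead use steady spawn rates, backward velocity to simulate motion, and ignore time of day.
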NiagaraSystem /BRCosmetics/Animation/Game/MainPlayer/Emotes/PizzaParty/FX/NS_Emote_PizzaParty_WheelLights
dependencies {
 usedDataInterfaces [
  &NiagaraDataInterfaceCamera
  &NiagaraDataInterfaceCurve
  &NiagaraDataInterfaceColorCurve
  &NiagaraDataInterfaceSkeletalMesh
 ]
 usedDynamicInputs [/Niagara/DynamicInputs/Math/Min_Float, &Subtract_Float]
 usedModules [
  &self:SystemSpawnScript.SystemScriptSource.SystemScriptGraph.NiagaraNodeAssignment_3.SetVariables_#1
  &SystemState
  &FN_Timer
  &self:SystemSpawnScript.SystemScriptSource.SystemScriptGraph.NiagaraNodeAssignment_4.SetVariables_#2
  &self:SystemSpawnScript.SystemScriptSource.SystemScriptGraph.NiagaraNodeAssignment_0.SetVariables_#3
  &Timeline
  &self:SystemSpawnScript.SystemScriptSource.SystemScriptGraph.NiagaraNodeAssignment_1.SetVariables_#4
  &self:SystemSpawnScript.SystemScriptSource.SystemScriptGraph.NiagaraNodeAssignment_5.SetVariables_#5
 ]
}
emitters {
 BackLight {
  emitterUpdateScript [
   &EmitterState(
    Inactive Response=::NewEnumerator1
    Life Cycle Mode=::NewEnumerator1
    Loop Behavior=::NewEnumerator1
    Loop Duration Mode=::NewEnumerator1
    UseLoopDelay=false
   )
   &SpawnBurst_Instantaneous(Use Spawn Probability=false, UseLoopCountLimit=false)
  ]
  particleSpawnScript [
   &FN_SetSpawnGroup(Number of Spawn Groups=1)
   &InitializeParticle(
    Color=<rgba 33 1.32 1.32 1>
    Mesh Scale Mode=::NewEnumerator3
    Sprite Size Mode=::NewEnumerator3
    Uniform Sprite Size=8
    UsePositionOffset=true
    Write Material Random=true
   )
   &SpriteFacingAndAlignment(Sprite Alignment=<0 0 1>, Sprite Facing=<0 -1 0>, WriteAlignment=true, WriteFacing=true)
   /SetVariables(
    Particles.GlowCameraOffset:NiagaraFloat=&FN_FrontEndFloatModifier(Gameplay=20, Front End=5)
    Particles.GlowPosition:NiagaraPosition=$Particles.Position
    Particles.GlowSize=&Multiply_Vector2D_ByFloat(Float=4, Vector2D=$Particles.SpriteSize)
   ) {isSetParameters=true}
   /SetVariables(Particles.FlareSize:Vector2f=<25 10>) {isSetParameters=true}
  ]
  particleUpdateScript [
   &ParticleState(Kill Particles When Lifetime Has Elapsed=false)
   &SkeletalMeshLocation(MaskSpawnGroup=false, Mesh Sampling Type=::NewEnumerator5, Skeletal Mesh=$System.Skeletal Mesh, Socket Sampling Mode=::NewEnumerator5)
   /SetVariables(
    Particles.GlowColor=&MultiplyLinearColor(
     Linear Color=&MultiplyLinearColor(
      Linear Color=<rgba 33 1.32 1.32 1>
      Scale Factor=&MakeVector4_FromFloat(Float=&Add_Float(
        A=0.6
        B=&Multiply_Float(
         A=0.5
         B=&ClampFloat(Float=&OneMinusFloat(Float=&Multiply_Float(
            A=3
            B=&VectorLength(
             Vector=/Niagara/DynamicInputs/Vectors/CrossProduct(
              Vector A:Vector3f=&ReturnCameraVector(Camera Vectors=::NewEnumerator1, Coordinate Space=::Simulation)
              Vector B:Vector3f=<0 1 0.2>
              Normalize Incoming Vectors:NiagaraBool=true
             )
             Fallback Vector=<0 -1 0>
            )
           )))
        )
       ))
     )
     Scale Factor=&MakeVector4_FromFloat(Float=&Multiply_Float(
       A=&FN_FrontEndFloatModifier(Gameplay=0.08, Front End=0.16)
       B=&FN_FrontEndFloatModifier(
        Gameplay=&Multiply_Float(
         A=&FNTimeOfDayScalar(Night Scale=1, Morning Scale=0, Day Scale=0, Evening Scale=0)
         B=$System.FadeIn
        )
        Front End=$System.FadeIn
       )
      ))
    )
    Particles.GlowSize=&Multiply_Vector2D_ByFloat(
     Float=&MakeFloatFromVector(Vector=$Engine.Owner.Scale, Channel=::NewEnumerator0)
     Vector2D=$Particles.Initial.GlowSize
    )
    Particles.LightColor=&MultiplyLinearColor(
     Linear Color=<rgba 20 1.32 1.32 1>
     Scale Factor=&MakeVector4(
      X=1
      Y=1
      Z=1
      W=&Multiply_Float(
       A=$System.FadeIn
       B=&FNTimeOfDayScalar(Night Scale=1, Morning Scale=0, Day Scale=0, Evening Scale=0)
      )
     )
    )
    Particles.LightPos:NiagaraPosition=&AddVectorToPosition(
     Position=$Particles.Position
     Vector=&RotateVectorByQuaternion(Vector To Rotate=<0 8 0>, Quaternion=$Particles.SkeletalMeshLocation.SampledRotation)
    )
    Particles.LightRadius=35
    Particles.SpriteSize=&Multiply_Vector2D_ByFloat(
     Float=&MakeFloatFromVector(Vector=$Engine.Owner.Scale, Channel=::NewEnumerator0)
     Vector2D=$Particles.Initial.SpriteSize
    )
   ) {isSetParameters=true}
   &ScaleColor(
    Scale Alpha=&Multiply_Float(
     A=&FN_FrontEndFloatModifier(
      Gameplay=&FNTimeOfDayScalar(Night Scale=1, Morning Scale=0, Day Scale=0, Evening Scale=0)
      Front End=$System.FadeIn
     )
     B=$System.FadeIn
    )
    ScaleA=true
    ScaleRGB=true
   ) {module=ScaleColor002}
  ]
  rendererClasses [&NiagaraSpriteRendererProperties, &NiagaraLightRendererProperties]
  renderers [
   {rendererClass=&NiagaraSpriteRendererProperties}
   {rendererClass=&NiagaraLightRendererProperties, rendererIndex=1}
   {rendererClass=&NiagaraSpriteRendererProperties, rendererIndex=2}
  ]
 }
 Directional_Smoke_Ground001 {
  emitterUpdateScript [
   &EmitterState()
   &SpawnRate(
    Spawn Probability=&FN_FrontEndFloatModifier(Gameplay=0, Front End=1)
    SpawnRate=10
    Use Spawn Probability=true
   )
   &SpawnPerUnit(
    Module.Use Max Movement Threshold=true
    Module.Use Movement Tolerance=false
    Spawn Probability:NiagaraFloat=&FN_FrontEndFloatModifier(
     Gameplay=&MakeCustomFloatFromBool(
      Bool=&SetBoolByFloatComparison(A=&VectorLength(Vector=$Engine.Owner.Velocity), B=150)
      True Float=1
      False Float=0
     )
     Front End=0
    )
    Use Spawn Probability=true
   )
  ]
  particleSpawnScript [
   &InitializeParticle(
    Color=<rgba 1 1 1 0.05>
    Lifetime=&RandomRangeFloat(Minimum=0.5, Maximum=0.6)
    Sprite Rotation Mode=::NewEnumerator2
    Sprite Size Mode=::NewEnumerator1
    Uniform Sprite Size Max=60
    Uniform Sprite Size Min=50
    UsePositionOffset=false
    Write Material Random=true
   )
   &ShapeLocation(
    Cylinder Distribution=::NewEnumerator2
    Cylinder Height=0
    Cylinder Radius=5
    MaskSpawnGroup=false
    Offset Mode=::NewEnumerator0
    Rotation Coordinate Space=::World
    Shape Primitive=::NewEnumerator1
    U Position=&ReturnNormalizedExecIndex(Normalized Index Includes One=false)
    UseOwnerScale=true
   )
   &AddVelocity(Velocity=&FN_FrontEndVector3Modifier(Gameplay=<1 0 0>, Front End=<0 -200 0>)) {module=AddVelocity001}
   &ApplyInitialForces()
   /SetVariables(Particles.FirstFlipBookFrame:NiagaraInt32=&RandomRangeInt(Minimum=0, Maximum=64)) {isSetParameters=true}
  ]
  particleUpdateScript [
   &ParticleState()
   &ScaleSpriteSize(Uniform Curve Sprite Scale=DataInterface() {propertyValues {Curve {keys [
        {arriveTangent=10.1767, interpMode=RCIM_Cubic, leaveTangent=10.1767, tangentMode=RCTM_User, time=0, value=0.2}
        {interpMode=RCIM_Cubic, tangentMode=RCTM_User, time=0.2, value=1}
        {arriveTangent=-0.6666, interpMode=RCIM_Cubic, leaveTangent=-0.6666, tangentMode=RCTM_User, time=1, value=0.3}
       ]}}})
   &DynamicMaterialParameters(
    Index 0 Param 1=&FloatFromCurve(
     FloatCurve=DataInterface() {propertyValues {Curve {keys [
         {arriveTangent=2.028, interpMode=RCIM_Cubic, leaveTangent=2.028, tangentMode=RCTM_User, time=0, value=0}
         {arriveTangent=0.1368, interpMode=RCIM_Cubic, leaveTangent=0.1368, tangentMode=RCTM_User, time=1, value=1}
        ]}}}
     Scale Curve=0.5
    )
    Index 0 Param 4=0.2
    Module.Param0WriteEnabled=true
    Module.Param1WriteEnabled=false
    Module.Param2WriteEnabled=false
    Module.Param3WriteEnabled=true
   ) {module=DynamicMatParams}
   &ScaleColor(
    Scale Alpha=&Multiply_Float(
     A=&FloatFromCurve(FloatCurve=DataInterface() {propertyValues {Curve {keys [
          {interpMode=RCIM_Cubic, leaveTangent=5.1035, tangentMode=RCTM_Break, time=0, value=0}
          {arriveTangent=-0.2156, interpMode=RCIM_Cubic, leaveTangent=-0.2156, tangentMode=RCTM_User, time=0.3, value=0.8}
          {arriveTangent=-0.0478, interpMode=RCIM_Cubic, leaveTangent=-0.0478, tangentMode=RCTM_User, time=1, value=0}
         ]}}})
     B=&FN_FrontEndFloatModifier(
      Gameplay=&FNTimeOfDayScalar(Night Scale=1, Morning Scale=0.4, Day Scale=0.8, Evening Scale=0.3)
      Front End=1
     )
    )
    Scale RGB=&MakeVectorFromLinearColorRGB(Color=&ColorFromCurve(ColorCurve=DataInterface() {propertyValues {
        AlphaCurve {keys [<at 0.3955 1 RCIM_Cubic>]}
        BlueCurve {keys [<at 0 0.2111 RCIM_Cubic>]}
        GreenCurve {keys [<at 0 0.3777 RCIM_Cubic>]}
        RedCurve {keys [<at 0 0.5104 RCIM_Cubic>]}
       }}))
    ScaleA=true
    ScaleRGB=true
   )
   &SubUVAnimation(
    End Frame:NiagaraInt32=&Add_Integer(A=$Particles.FirstFlipBookFrame, B=&Multiply_Int(A=&ReturnExecIndex(), B=8))
    Start Frame:NiagaraInt32=$Particles.FirstFlipBookFrame
    SubUV Loop Count=0
   )
   &SolveForcesAndVelocity(Clamp Velocity=false, Limit Acceleration=false)
   &CameraOffset(Camera Offset Amount=15)
  ]
  rendererClasses [&NiagaraSpriteRendererProperties]
  renderers [
   {rendererClass=&NiagaraSpriteRendererProperties}
   {rendererClass=&NiagaraSpriteRendererProperties, rendererIndex=1}
  ]
 }
 Headlights001 {
  emitterUpdateScript [
   &EmitterState(
    Inactive Response=::NewEnumerator1
    Life Cycle Mode=::NewEnumerator1
    Loop Behavior=::NewEnumerator1
    Loop Duration Mode=::NewEnumerator1
    UseLoopDelay=false
   )
   &SpawnBurst_Instantaneous(Use Spawn Probability=false, UseLoopCountLimit=false)
  ]
  particleSpawnScript [
   &FN_SetSpawnGroup(Number of Spawn Groups=2)
   &InitializeParticle(
    Color=<rgba 0.417 0.6016 1 0.1>
    Mesh Scale Mode=::NewEnumerator3
    Sprite Size=<100 100>
    Sprite Size Mode=::NewEnumerator4
    UsePositionOffset=true
    Write Material Random=true
   )
   &CameraOffset()
   &DynamicMaterialParameters(
    Index 0 Param 1=&Multiply_Float(A=-19, B=&MakeFloatFromVector(Vector=$Particles.MirrorAxis, Channel=::NewEnumerator0))
    Index 0 Param 2=-1
    Module.Param0WriteEnabled=true
    Module.Param1WriteEnabled=true
    Module.Param2WriteEnabled=false
    Module.Param3WriteEnabled=false
   )
  ]
  particleUpdateScript [
   &ParticleState(Kill Particles When Lifetime Has Elapsed=false)
   &SkeletalMeshLocation(
    Bone / Socket Index=1
    MaskSpawnGroup=false
    Mesh Sampling Type=::NewEnumerator5
    Orientation Sampling=::NewEnumerator1
    Skeletal Mesh=$System.Skeletal Mesh
    Socket Sampling Mode=::NewEnumerator5
   )
   /SetVariables(
    Particles.GlowColor=&MultiplyLinearColor(
     Linear Color=$Particles.Color
     Scale Factor=&MakeVector4_FromFloat(Float=&Multiply_Float(
       A=&Multiply_Float(A=8, B=$System.FadeIn)
       B=&ClampFloat(Float=&OneMinusFloat(Float=&Multiply_Float(
          A=0.7
          B=&VectorLength(
           Vector=&CrossProduct(
            Vector A:Vector3f=&ReturnCameraVector(Camera Vectors=::NewEnumerator1, Coordinate Space=::Simulation)
            Vector B:Vector3f=<0 -1 0>
            Normalize Incoming Vectors:NiagaraBool=true
           )
           Fallback Vector=<0 -1 0>
          )
         )))
      ))
    )
    Particles.GlowPosition:NiagaraPosition=$Particles.Position
    Particles.GlowSize=&Vector2DFromFloat(Value=15)
    Particles.glowalignement:Vector3f=<1 0 0>
    Particles.glowfacing:Vector3f=<0 1 0>
   ) {isSetParameters=true}
   /SetVariables(
    Particles.BigGlowColor:LinearColor=&MultiplyLinearColor(
     Linear Color=$Particles.Color
     Scale Factor=&MakeVector4_FromFloat(Float=&Multiply_Float(
       A=&Multiply_Float(A=1.5, B=$System.FadeIn)
       B=&ClampFloat(Float=&OneMinusFloat(Float=&Multiply_Float(
          A=0.4
          B=&VectorLength(
           Vector=&CrossProduct(
            Vector A:Vector3f=&ReturnCameraVector(Camera Vectors=::NewEnumerator1, Coordinate Space=::Simulation)
            Vector B:Vector3f=<0 -1 0>
            Normalize Incoming Vectors:NiagaraBool=true
           )
           Fallback Vector=<0 -1 0>
          )
         )))
      ))
    )
    Particles.BigGlowSize:Vector2f=&Vector2DFromFloat(Value=60)
    Particles.GlowCameraOffset:NiagaraFloat=&Lerp_Float(
     A=-35
     B=-5
     Alpha=&Power(
      a=&DotProduct(
       Vector A=&NormalizeVector(
        Vector To Normalize=&ReturnCameraVector(Camera Vectors=::NewEnumerator1, Coordinate Space=::World)
        Safe Normalize=true
        Safe Fallback Threshold=0
        Safe Fallback Vector=<0 0 0.001>
       )
       Vector B=&NormalizeVector(Vector To Normalize=$Engine.Owner.SystemXAxis, Safe Normalize=true, Safe Fallback Threshold=0, Safe Fallback Vector=<0 0 0.001>)
      )
      b=3
     )
    )
   ) {isSetParameters=true}
   /SetVariables(
    Particles.LightColor=&MultiplyLinearColor(
     Linear Color=$Particles.Color
     Scale Factor=&MakeVector4_FromFloat(Float=&Multiply_Float(A=4, B=$System.FadeIn))
    )
    Particles.LightPosition:NiagaraPosition=$Particles.Position
   ) {isSetParameters=true}
   /SetVariables(
    Particles.TimeOfDayHeadlightsIntensity:NiagaraFloat=&Lerp_Float(
     A=&Multiply_Float(
      A=$System.FadeIn
      B=&FN_FrontEndFloatModifier(
       Gameplay=&FNTimeOfDayScalar(Night Scale=1, Morning Scale=0, Day Scale=0, Evening Scale=0)
       Front End=1
      )
     )
     B=1
     Alpha=1
    )
    Particles.TimeOfDayVisibilityTag:NiagaraInt32=&MakeIntFromFloat(Float=&FN_FrontEndFloatModifier(Gameplay=&FNTimeOfDayScalar(Night Scale=0, Morning Scale=1, Evening Scale=0), Front End=0))
   ) {enabled=false, isSetParameters=true}
   &SpriteFacingAndAlignment(
    Alignment Coordinate Space=::Local
    Sprite Alignment=&Multiply_Vector(A=<-0.05 0.5 0.1>, B=$Particles.MirrorAxis)
    WriteAlignment=true
    WriteFacing=false
   ) {enabled=false}
   &SpriteFacingAndAlignment(
    Alignment Coordinate Space=::Local
    Sprite Alignment=<0 -1 0>
    Sprite Facing=&RotateVectorByQuaternion(Vector To Rotate=<0 0 0>, Quaternion=$Particles.SkeletalMeshLocation.SampledRotation)
    WriteAlignment=true
    WriteFacing=false
   ) {module=SpriteFacingAndAlignment001}
   &ScaleColor(
    Scale Alpha=&Multiply_Float(
     A=&FN_FrontEndFloatModifier(
      Gameplay=&FNTimeOfDayScalar(Night Scale=1, Morning Scale=0, Day Scale=0, Evening Scale=0)
      Front End=1
     )
     B=1
    )
    ScaleA=true
    ScaleRGB=true
   )
   &ScaleColor(Scale Alpha=$System.FadeIn, ScaleA=true, ScaleRGB=true) {module=ScaleColor002}
  ]
  rendererClasses [&NiagaraSpriteRendererProperties, &NiagaraLightRendererProperties]
  renderers [
   {rendererClass=&NiagaraSpriteRendererProperties}
   {rendererClass=&NiagaraLightRendererProperties, rendererIndex=1}
   {rendererClass=&NiagaraSpriteRendererProperties, rendererIndex=2}
   {rendererClass=&NiagaraSpriteRendererProperties, rendererIndex=3}
  ]
 }
 LightTrail {
  emitterUpdateScript [
   &EmitterState()
   &SpawnRate(
    Spawn Probability=&FN_FrontEndFloatModifier(Gameplay=0, Front End=1)
    SpawnRate=40
    Use Spawn Probability=true
   )
   &SpawnPerUnit(
    Module.Use Max Movement Threshold=true
    Module.Use Movement Tolerance=false
    Spawn Probability:NiagaraFloat=&FN_FrontEndFloatModifier(Gameplay=1, Front End=0)
    Spawn Spacing=10
    Use Spawn Probability=true
   )
  ]
  particleSpawnScript [
   &InitializeParticle(
    Color=<rgba 1 0 0 0.05>
    Lifetime=0.25
    Ribbon Width=10
    Ribbon Width Mode=::NewEnumerator1
    UsePositionOffset=false
    Write Material Random=true
   )
   &SkeletalMeshLocation(
    MaskSpawnGroup=false
    Mesh Sampling Type=::NewEnumerator5
    Sampled Position Offset=<0 -10 1.6>
    Skeletal Mesh=$System.Skeletal Mesh
    Socket Sampling Mode=::NewEnumerator5
   )
   &AddVelocity(Velocity=<0 -300 0>, Velocity Speed Scale=&FN_FrontEndFloatModifier(Gameplay=0, Front End=1))
  ]
  particleUpdateScript [
   &ParticleState()
   &ScaleColor(
    Scale Alpha=&Multiply_Float(
     A=&Multiply_Float(
      A=&FloatFromCurve(FloatCurve=DataInterface() {propertyValues {Curve {keys [<at 0 0 RCIM_Cubic>, <at 0.25 1 RCIM_Cubic>, {interpMode=RCIM_Cubic, tangentMode=RCTM_User, time=1, value=0}]}}})
      B=&FN_FrontEndFloatModifier(
       Gameplay=&FNTimeOfDayScalar(Night Scale=1, Morning Scale=0, Day Scale=0, Evening Scale=0)
       Front End=1
      )
     )
     B=$System.FadeIn
    )
    Scale RGB=&Multiply_VectorByFloat(Vector=<1 1 1>, Float=&FN_FrontEndFloatModifier(Gameplay=1, Front End=2))
    ScaleA=true
    ScaleRGB=true
   )
   &SolveForcesAndVelocity(Clamp Velocity=false, Limit Acceleration=false)
   &ScaleRibbonWidth(Ribbon Width Scale=&FloatFromCurve(FloatCurve=DataInterface() {propertyValues {Curve {keys [{tangentMode=RCTM_User, time=0, value=0}, {tangentMode=RCTM_User, time=0.25, value=1}]}}}))
  ]
  rendererClasses [&NiagaraRibbonRendererProperties]
  renderers [{rendererClass=&NiagaraRibbonRendererProperties}]
 }
}
summary {
 systemName=NS_Emote_PizzaParty_WheelLights
 emitters [
  {emitterName=BackLight, rendererClasses [&NiagaraSpriteRendererProperties, &NiagaraLightRendererProperties]}
  {emitterName=Headlights001, rendererClasses [&NiagaraSpriteRendererProperties, &NiagaraLightRendererProperties]}
  {emitterName=LightTrail, rendererClasses [&NiagaraRibbonRendererProperties]}
  {emitterName=Directional_Smoke_Ground001, rendererClasses [&NiagaraSpriteRendererProperties]}
 ]
 userVariables [
  User.bIsFrontEnd {type=NiagaraBool}
  User.bIsFrontEndPreview {type=NiagaraBool}
  User.ForwardDot {type=NiagaraFloat, defaultValue=0, description=""}
 ]
}
